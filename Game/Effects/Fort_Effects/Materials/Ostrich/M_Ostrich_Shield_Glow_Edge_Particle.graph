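# Unlit translucent, two-sided mesh particle material for the glowing, electrical-looking edge of the Ostrich shield effect.
# Particle color becomes time-of-day-adjusted HDR emissive, while opacity combines a U gradient, particle alpha, near-camera response and depth fade.
# World position offset combines SplineThicken width, stepped-time panning noise jitter and vector noise, with dynamic parameters scaling thickness and jitter.
# Mobile and forward paths simplify color and skip vector noise, as does low quality for the noise.
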Material /Game/Effects/Fort_Effects/Materials/Ostrich/M_Ostrich_Shield_Glow_Edge_Particle
graph {
 properties {
  bUsedWithMeshDeformer=true
  bUsedWithMeshParticles=true
  bUsedWithSkeletalMesh=true
  blendMode=BLEND_Translucent
  pixelDepthOffsetMode=PDOM_Legacy
  previewMesh=/Game/Effects/Fort_Effects/Gameplay/Ostrich/SM_Ostrich_Shield_Return_BottomEdge_Electrical
  shadingModel=MSM_Unlit
  twoSided=true
  usedShadingModels=Unlit
 }
 nodes [
  n1 = ComponentMask(None=TextureCoordinate()) {r=true}
  n2 = Constant() {r=1}
  n3 = Constant() {r=1.5}
  n4 = ScalarParameter() {defaultValue=500, parameterName=NearFade}
  n5 = Clamp(None=Power(Base=MaterialFunctionCall(Distance=@n4) {materialFunction=&Near_Camera_Fade}.Result) {constExponent=16})
  n6 = Multiply(A=Multiply(A=OneMinus(None=@n1), B=LinearInterpolate(A=@n2, B=@n3, Alpha=@n5)), B=ParticleColor().A)
  n7 = DepthFade(Opacity=@n6, FadeDistance=ScalarParameter() {defaultValue=25, parameterName=Depth_FadeDistance})
  FeatureLevelSwitch(Default=@n7, ES2=@n6, ES3_1=@n6, SM4_REMOVED=@n7)
  n8 = ParticleColor()
  n9 = MaterialFunctionCall(LDR Input=VectorParameter() {parameterName=LDR_Input_Color, defaultValue=<rgba 1 1 1 1>}.RGB, HDR Tint=@n8.RGB) {materialFunction=&DeriveHDRfromLDR}
  n10 = MaterialFunctionCall(
   Color_Morning=@n9.RemappedRange
   Color_Day=@n9.RemappedRange
   Color_Evening=@n9.RemappedRange
   Color_Night=Multiply(A=@n9.RemappedRange, B=ScalarParameter() {defaultValue=0.5, parameterName=TOD_Night_ColorMult})
  ) {materialFunction=&Color_Time_of_Day}
  n11 = FeatureLevelSwitch(Default=@n10.Result, ES2=@n8.RGB, ES3_1=@n8.RGB, SM4_REMOVED=@n10.Result)
  n12 = ScalarParameter() {defaultValue=0.25, parameterName=Emissive_Mult}
  n13 = ScalarParameter() {defaultValue=3, parameterName=Emissive_Mult_ForMobileAndSwitch}
  n14 = FeatureLevelSwitch(Default=@n12, ES2=@n13, ES3_1=@n13, SM4_REMOVED=@n12)
  MaterialFunctionCall(Emissive=Multiply(A=ShadingPathSwitch(Default=@n11, Forward=@n8.RGB), B=ShadingPathSwitch(Default=@n14, Forward=@n13))) {materialFunction=&MF_FX_EmissiveAdjustments}
  Multiply()
  n15 = Subtract(A=ComponentMask(None=TextureCoordinate()) {g=true}, B=Constant()) {constB=0.5}
  n16 = Clamp(None=Multiply(A=@n15, B=OneMinus(None=@n15)))
  n17 = Add(A=TextureSample(UVs=TextureCoordinate() {coordinateIndex=1, vTiling=2}) {texture=/Game/Effects/Fort_Effects/Textures/Ramps/T_Ramp_Smooth_Reverse}.G) {constB=0.6}
  n18 = ScalarParameter() {defaultValue=25, parameterName=SplineThickness}
  n19 = DynamicParameter() {paramNames [Thickness,Noise,Param3,Param4]}
  n20 = Power(Base=MaterialFunctionCall(Distance=ScalarParameter() {defaultValue=500, parameterName=NearFade}) {materialFunction=&Near_Camera_Fade}.Result)
  n21 = Multiply(
   A=Multiply(A=@n16, B=Clamp(None=@n17))
   B=LinearInterpolate(A=ScalarParameter() {parameterName=SplineThickness_Close}, B=Multiply(A=@n18, B=@n19.Thickness), Alpha=Clamp(None=@n20))
  )
  n22 = MaterialFunctionCall(WidthBase=@n21, WidthTip=@n21, UVs for Thickness=TextureCoordinate()) {materialFunction=&SplineThicken}
  n23 = TextureCoordinate() {uTiling=0.002, vTiling=0.1}
  n24 = ScalarParameter() {defaultValue=5, parameterName=TileWpo_Noise}
  n25 = Time()
  n26 = ScalarParameter() {defaultValue=25, parameterName=PanStepSize}
  n27 = Divide(A=Round(None=Multiply(A=@n25, B=@n26)), B=@n26)
  n28 = Constant2Vector() {g=1, r=0.25}
  n29 = ScalarParameter() {defaultValue=-1, parameterName=PanSpeed}
  n30 = TextureSample(UVs=Panner(Coordinate=Multiply(A=@n23, B=@n24), Time=@n27, Speed=Multiply(A=@n28, B=@n29)) {bFractionalPart=true, speedX=0.01, speedY=0.1}) {samplerType=SAMPLERTYPE_LinearColor, texture=/Game/Global/Textures/Default/Noise/Noise32x32}
  n31 = OneMinus(None=@n30.RGB)
  n32 = ScalarParameter() {defaultValue=1, parameterName=WPO_Noise_Mult_max}
  n33 = DynamicParameter() {paramNames [Thickness,Noise,Param3,Param4]}
  n34 = Panner(
   Coordinate=Multiply(A=TextureCoordinate() {vTiling=0.1}, B=Multiply(A=@n24) {constB=6}) {constB=0.1}
   Time=Multiply(A=@n27) {constB=0.5}
   Speed=Multiply(A=@n29) {constB=-1.43}
  ) {bFractionalPart=true, speedX=-0.05}
  n35 = Multiply(
   A=Multiply(A=Subtract(A=@n30.RGB, B=@n31), B=Multiply(A=@n32, B=@n33.Noise)) {constA=1}
   B=Add(A=TextureSample(UVs=@n34) {texture=/Game/Global/Textures/Default/Masks_and_Gradients/T_Vert_Gradient_Sections}.R) {constB=0.15}
  )
  n36 = WorldPosition()
  n37 = ActorPositionWS()
  n38 = ScalarParameter() {defaultValue=10, parameterName=CurlSpeed}
  n39 = Time()
  n40 = Divide(A=Subtract(A=Subtract(A=@n36.XYZ, B=@n37), B=Multiply(A=@n38, B=@n39)), B=ScalarParameter() {defaultValue=15, parameterName=CurlScale})
  n41 = Multiply(
   A=@n16
   B=Multiply(A=ScalarParameter() {defaultValue=10, parameterName=CurlIntensity}, B=VectorNoise(World Position=@n40) {noiseFunction=VNF_VectorALU})
  )
  n42 = ShadingPathSwitch(Default=FeatureLevelSwitch(Default=@n41, ES2=@n16, ES3_1=@n16, SM4_REMOVED=@n41), Forward=@n16)
  Reroute(None=Add(A=Add(A=@n22.WorldPositionOffset, B=@n35), B=QualitySwitch(Default=@n42, Low=@n16)))
 ]
}
